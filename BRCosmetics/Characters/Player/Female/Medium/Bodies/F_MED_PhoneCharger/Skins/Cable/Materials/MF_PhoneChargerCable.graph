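# Material function for the PhoneCharger Cable character skin that adds music-reactive emissive glow to incoming Material Attributes.
# Audio MPC spectrum channels fill four sections of a local-bounds Z gradient; amplitude feeds effect modulation.
# Animated sine lines over a distance-field texture, cutoff Fresnel glow and time-varying random boosts form the glow masks.
# Tinted glow and scaled incoming emissive are summed, then blended in through masks, changing only EmissiveColor.
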MaterialFunction /BRCosmetics/Characters/Player/Female/Medium/Bodies/F_MED_PhoneCharger/Skins/Cable/Materials/MF_PhoneChargerCable
graph {expressions {
  MaterialExpressionAdd_0 {
   class=MaterialExpressionAdd
   inputs [
    A {from=MaterialExpressionAdd_4}
    B {from=MaterialExpressionAdd_24}
   ]
  }
  MaterialExpressionAdd_1 {
   class=MaterialExpressionAdd
   inputs [
    A {from=MaterialExpressionMultiply_2}
    B {from=MaterialExpressionMultiply_42}
   ]
  }
  MaterialExpressionAdd_10 {
   class=MaterialExpressionAdd
   inputs [
    A {from=MaterialExpressionMultiply_51}
    B {from=MaterialExpressionAdd_29}
   ]
  }
  MaterialExpressionAdd_14 {
   class=MaterialExpressionAdd
   inputs [
    A {from=MaterialExpressionAdd_20}
    B {from=MaterialExpressionScalarParameter_12}
   ]
  }
  MaterialExpressionAdd_2 {
   class=MaterialExpressionAdd
   inputs [
    A {from=MaterialExpressionScalarParameter_4}
    B {from=MaterialExpressionScalarParameter_5}
   ]
  }
  MaterialExpressionAdd_20 {
   class=MaterialExpressionAdd
   inputs [
    A {from=MaterialExpressionScalarParameter_7}
    B {from=MaterialExpressionMultiply_9}
   ]
  }
  MaterialExpressionAdd_21 {
   class=MaterialExpressionAdd
   inputs [
    A {from=MaterialExpressionAdd_31}
    B {from=MaterialExpressionNamedRerouteUsage_74}
   ]
  }
  MaterialExpressionAdd_22 {
   class=MaterialExpressionAdd
   inputs [
    A {from=MaterialExpressionScalarParameter_21}
    B {from=""}
   ]
   properties {constB=0.01}
  }
  MaterialExpressionAdd_23 {
   class=MaterialExpressionAdd
   inputs [
    A {from=MaterialExpressionScalarParameter_23}
    B {from=""}
   ]
   properties {constB=0.01}
  }
  MaterialExpressionAdd_24 {
   class=MaterialExpressionAdd
   inputs [
    A {from=MaterialExpressionScalarParameter_36}
    B {from=MaterialExpressionMultiply_0}
   ]
  }
  MaterialExpressionAdd_29 {
   class=MaterialExpressionAdd
   inputs [
    A {from=MaterialExpressionMultiply_24}
    B {from=MaterialExpressionMultiply_29}
   ]
  }
  MaterialExpressionAdd_3 {
   class=MaterialExpressionAdd
   inputs [
    A {from=MaterialExpressionMultiply_4}
    B {from=MaterialExpressionMultiply_33}
   ]
  }
  MaterialExpressionAdd_31 {
   class=MaterialExpressionAdd
   inputs [
    A {from=MaterialExpressionNamedRerouteUsage_40}
    B {from=MaterialExpressionNamedRerouteUsage_64}
   ]
  }
  MaterialExpressionAdd_34 {
   class=MaterialExpressionAdd
   inputs [
    A {from=MaterialExpressionAdd_9}
    B {from=MaterialExpressionMultiply_48}
   ]
  }
  MaterialExpressionAdd_4 {
   class=MaterialExpressionAdd
   inputs [
    A {from=MaterialExpressionMultiply_1}
    B {from=MaterialExpressionMultiply_6}
   ]
  }
  MaterialExpressionAdd_5 {
   class=MaterialExpressionAdd
   inputs [
    A {from=MaterialExpressionMaterialFunctionCall_11, output=Local Bounds Minimum}
    B {from=MaterialExpressionScalarParameter_14}
   ]
  }
  MaterialExpressionAdd_6 {
   class=MaterialExpressionAdd
   inputs [
    A {from=MaterialExpressionMaterialFunctionCall_11, output=Local Bounds Max}
    B {from=MaterialExpressionScalarParameter_15}
   ]
  }
  MaterialExpressionAdd_9 {
   class=MaterialExpressionAdd
   inputs [
    A {from=MaterialExpressionAdd_1}
    B {from=MaterialExpressionMultiply_3}
   ]
  }
  MaterialExpressionCollectionParameter_1 {
   class=MaterialExpressionCollectionParameter
   inputs []
   properties {collection=/Game/Athena/Events/AudioAnalysis/MPC_Audio_Reactive_Parameters, parameterName=Spectrum_Music}
  }
  MaterialExpressionCollectionParameter_2 {
   class=MaterialExpressionCollectionParameter
   inputs []
   properties {collection=&MPC_Audio_Reactive_Parameters, parameterName=Amplitude_Music}
  }
  MaterialExpressionComponentMask_1 {class=MaterialExpressionComponentMask, inputs [None {from=MaterialExpressionNamedRerouteUsage_14}], properties {r=true}}
  MaterialExpressionComponentMask_11 {class=MaterialExpressionComponentMask, inputs [None {from=MaterialExpressionNamedRerouteUsage_71}], properties {g=true}}
  MaterialExpressionComponentMask_2 {class=MaterialExpressionComponentMask, inputs [None {from=MaterialExpressionNamedRerouteDeclaration_40}], properties {r=true}}
  MaterialExpressionComponentMask_3 {class=MaterialExpressionComponentMask, inputs [None {from=MaterialExpressionNamedRerouteUsage_42}], properties {g=true}}
  MaterialExpressionComponentMask_5 {class=MaterialExpressionComponentMask, inputs [None {from=MaterialExpressionDivide_3}], properties {b=true}}
  MaterialExpressionComponentMask_6 {class=MaterialExpressionComponentMask, inputs [None {from=MaterialExpressionNamedRerouteUsage_0}], properties {b=true}}
  MaterialExpressionComponentMask_8 {class=MaterialExpressionComponentMask, inputs [None {from=MaterialExpressionNamedRerouteUsage_2}], properties {a=true}}
  MaterialExpressionConstantBiasScale_0 {class=MaterialExpressionConstantBiasScale, inputs [None {from=MaterialExpressionSine_0}]}
  MaterialExpressionConstantBiasScale_2 {class=MaterialExpressionConstantBiasScale, inputs [None {from=MaterialExpressionSine_2}]}
  MaterialExpressionDivide_3 {
   class=MaterialExpressionDivide
   inputs [
    A {from=MaterialExpressionSubtract_4}
    B {from=MaterialExpressionSubtract_1}
   ]
  }
  MaterialExpressionFresnel_0 {
   class=MaterialExpressionFresnel
   inputs [
    ExponentIn {from="", output=""}
    BaseReflectFractionIn {from="", output=""}
    Normal {from="", output=""}
   ]
   properties {baseReflectFraction=0, exponent=1}
  }
  MaterialExpressionFunctionInput_0 {
   class=MaterialExpressionFunctionInput
   inputs [Preview {from="", output=""}]
   properties {inputName=Material, inputType=FunctionInput_MaterialAttributes}
  }
  MaterialExpressionFunctionOutput_0 {class=MaterialExpressionFunctionOutput, inputs [None {from=MaterialExpressionNamedRerouteDeclaration_15, output=""}]}
  MaterialExpressionGetMaterialAttributes_1 {
   class=MaterialExpressionGetMaterialAttributes
   inputs [None {from=MaterialExpressionNamedRerouteDeclaration_14, output=""}]
   properties {attributeGetTypes [#1]}
  }
  MaterialExpressionLinearInterpolate_0 {
   class=MaterialExpressionLinearInterpolate
   inputs [
    A {from=MaterialExpressionNamedRerouteUsage_19}
    B {from=MaterialExpressionNamedRerouteUsage_23}
    Alpha {from=MaterialExpressionMultiply_56}
   ]
  }
  MaterialExpressionMaterialFunctionCall_11 {
   class=MaterialExpressionMaterialFunctionCall
   inputs []
   properties {materialFunction=&ObjectLocalBounds}
  }
  MaterialExpressionMultiply_0 {
   class=MaterialExpressionMultiply
   inputs [
    A {from=MaterialExpressionNamedRerouteUsage_12}
    B {from=MaterialExpressionScalarParameter_0}
   ]
  }
  MaterialExpressionMultiply_1 {
   class=MaterialExpressionMultiply
   inputs [
    A {from=MaterialExpressionTime_0}
    B {from=MaterialExpressionScalarParameter_1}
   ]
  }
  MaterialExpressionMultiply_10 {
   class=MaterialExpressionMultiply
   inputs [
    A {from=MaterialExpressionMultiply_12}
    B {from=MaterialExpressionScalarParameter_9}
   ]
  }
  MaterialExpressionMultiply_12 {
   class=MaterialExpressionMultiply
   inputs [
    A {from=MaterialExpressionMultiply_5}
    B {from=MaterialExpressionNamedRerouteUsage_17}
   ]
  }
  MaterialExpressionMultiply_2 {
   class=MaterialExpressionMultiply
   inputs [
    A {from=MaterialExpressionComponentMask_2}
    B {from=MaterialExpressionNamedRerouteDeclaration_36}
   ]
  }
  MaterialExpressionMultiply_24 {
   class=MaterialExpressionMultiply
   inputs [
    A {from=MaterialExpressionTime_1}
    B {from=MaterialExpressionScalarParameter_19}
   ]
  }
  MaterialExpressionMultiply_26 {
   class=MaterialExpressionMultiply
   inputs [
    A {from=MaterialExpressionSmoothStep_1}
    B {from=MaterialExpressionOneMinus_6}
   ]
  }
  MaterialExpressionMultiply_29 {
   class=MaterialExpressionMultiply
   inputs [
    A {from=MaterialExpressionComponentMask_11}
    B {from=MaterialExpressionScalarParameter_57}
   ]
  }
  MaterialExpressionMultiply_3 {
   class=MaterialExpressionMultiply
   inputs [
    A {from=MaterialExpressionComponentMask_6}
    B {from=MaterialExpressionNamedRerouteDeclaration_37}
   ]
  }
  MaterialExpressionMultiply_33 {
   class=MaterialExpressionMultiply
   inputs [
    A {from=MaterialExpressionNamedRerouteUsage_11}
    B {from=MaterialExpressionScalarParameter_29}
   ]
  }
  MaterialExpressionMultiply_4 {
   class=MaterialExpressionMultiply
   inputs [
    A {from=MaterialExpressionNamedRerouteDeclaration_22}
    B {from=MaterialExpressionAdd_21}
   ]
  }
  MaterialExpressionMultiply_40 {
   class=MaterialExpressionMultiply
   inputs [
    A {from=MaterialExpressionSmoothStep_0}
    B {from=MaterialExpressionNamedRerouteUsage_49}
   ]
  }
  MaterialExpressionMultiply_42 {
   class=MaterialExpressionMultiply
   inputs [
    A {from=MaterialExpressionComponentMask_3}
    B {from=MaterialExpressionNamedRerouteDeclaration_35}
   ]
  }
  MaterialExpressionMultiply_44 {
   class=MaterialExpressionMultiply
   inputs [
    A {from=MaterialExpressionStep_5}
    B {from=MaterialExpressionSaturate_13}
   ]
  }
  MaterialExpressionMultiply_45 {
   class=MaterialExpressionMultiply
   inputs [
    A {from=MaterialExpressionMultiply_44}
    B {from=MaterialExpressionScalarParameter_43}
   ]
  }
  MaterialExpressionMultiply_46 {
   class=MaterialExpressionMultiply
   inputs [
    A {from=MaterialExpressionMultiply_45}
    B {from=MaterialExpressionNamedRerouteUsage_65}
   ]
  }
  MaterialExpressionMultiply_48 {
   class=MaterialExpressionMultiply
   inputs [
    A {from=MaterialExpressionComponentMask_8}
    B {from=MaterialExpressionNamedRerouteDeclaration_38}
   ]
  }
  MaterialExpressionMultiply_5 {
   class=MaterialExpressionMultiply
   inputs [
    A {from=MaterialExpressionSaturate_0}
    B {from=MaterialExpressionPower_0}
   ]
  }
  MaterialExpressionMultiply_51 {
   class=MaterialExpressionMultiply
   inputs [
    A {from=MaterialExpressionNamedRerouteUsage_28}
    B {from=MaterialExpressionScalarParameter_50}
   ]
  }
  MaterialExpressionMultiply_56 {
   class=MaterialExpressionMultiply
   inputs [
    A {from=MaterialExpressionNamedRerouteUsage_3}
    B {from=MaterialExpressionNamedRerouteUsage_76}
   ]
  }
  MaterialExpressionMultiply_6 {
   class=MaterialExpressionMultiply
   inputs [
    A {from=MaterialExpressionScalarParameter_6}
    B {from=MaterialExpressionNamedRerouteUsage_1}
   ]
  }
  MaterialExpressionMultiply_8 {
   class=MaterialExpressionMultiply
   inputs [
    A {from=MaterialExpressionNamedRerouteUsage_8}
    B {from=MaterialExpressionSmoothStep_11}
   ]
  }
  MaterialExpressionMultiply_9 {
   class=MaterialExpressionMultiply
   inputs [
    A {from=MaterialExpressionSaturate_10}
    B {from=MaterialExpressionScalarParameter_20}
   ]
  }
  MaterialExpressionNamedRerouteDeclaration_0 {
   class=MaterialExpressionNamedRerouteDeclaration
   inputs [None {from=MaterialExpressionCollectionParameter_2}]
   properties {name=Amplitude, nodeColor=<rgba 1 0.1406 0 1>}
  }
  MaterialExpressionNamedRerouteDeclaration_1 {
   class=MaterialExpressionNamedRerouteDeclaration
   inputs [None {from=MaterialExpressionSaturate_1}]
   properties {name=DFLines_Mask, nodeColor=<rgba 1 0 0.8906 1>}
  }
  MaterialExpressionNamedRerouteDeclaration_10 {
   class=MaterialExpressionNamedRerouteDeclaration
   inputs [None {from=MaterialExpressionGetMaterialAttributes_1, output=EmissiveColor}]
   properties {name=IN_Emissive, nodeColor=<rgba 1 0 0.375 1>}
  }
  MaterialExpressionNamedRerouteDeclaration_13 {
   class=MaterialExpressionNamedRerouteDeclaration
   inputs [None {from=MaterialExpressionAdd_3}]
   properties {name=OUT_Emissive, nodeColor=<rgba 0 1 0.3906 1>}
  }
  MaterialExpressionNamedRerouteDeclaration_14 {
   class=MaterialExpressionNamedRerouteDeclaration
   inputs [None {from=MaterialExpressionFunctionInput_0}]
   properties {name=IN_Material, nodeColor=<rgba 0.0078 0 1 1>}
  }
  MaterialExpressionNamedRerouteDeclaration_15 {
   class=MaterialExpressionNamedRerouteDeclaration
   inputs [None {from=MaterialExpressionSetMaterialAttributes_0}]
   properties {name=OUT_Material, nodeColor=<rgba 0.5469 1 0 1>}
  }
  MaterialExpressionNamedRerouteDeclaration_16 {
   class=MaterialExpressionNamedRerouteDeclaration
   inputs [None {from=MaterialExpressionTextureSample_1, output=G}]
   properties {name=RandomGrayscale, nodeColor=<rgba 0.3359 0 1 1>}
  }
  MaterialExpressionNamedRerouteDeclaration_18 {
   class=MaterialExpressionNamedRerouteDeclaration
   inputs [None {from=MaterialExpressionConstantBiasScale_2}]
   properties {name=RandomGrayscaleSine, nodeColor=<rgba 0.5469 1 0 1>}
  }
  MaterialExpressionNamedRerouteDeclaration_2 {
   class=MaterialExpressionNamedRerouteDeclaration
   inputs [None {from=MaterialExpressionComponentMask_5}]
   properties {name=GradientMask, nodeColor=<rgba 1 0.8672 0 1>}
  }
  MaterialExpressionNamedRerouteDeclaration_21 {
   class=MaterialExpressionNamedRerouteDeclaration
   inputs [None {from=MaterialExpressionFresnel_0}]
   properties {name=Fresnel, nodeColor=<rgba 0 1 0.2969 1>}
  }
  MaterialExpressionNamedRerouteDeclaration_22 {
   class=MaterialExpressionNamedRerouteDeclaration
   inputs [None {from=MaterialExpressionVectorParameter_0, output=RGB}]
   properties {name=EmissiveColor, nodeColor=<rgba 0.9688 1 0 1>}
  }
  MaterialExpressionNamedRerouteDeclaration_23 {
   class=MaterialExpressionNamedRerouteDeclaration
   inputs [None {from=MaterialExpressionMultiply_8}]
   properties {name=FresnelGlowMask, nodeColor=<rgba 0.8516 0 1 1>}
  }
  MaterialExpressionNamedRerouteDeclaration_29 {
   class=MaterialExpressionNamedRerouteDeclaration
   inputs [None {from=MaterialExpressionMultiply_46}]
   properties {name=RandomBoost, nodeColor=<rgba 0.6406 0 1 1>}
  }
  MaterialExpressionNamedRerouteDeclaration_3 {
   class=MaterialExpressionNamedRerouteDeclaration
   inputs [None {from=MaterialExpressionOneMinus_5}]
   properties {name=DistanceField, nodeColor=<rgba 1 0 0.5156 1>}
  }
  MaterialExpressionNamedRerouteDeclaration_30 {
   class=MaterialExpressionNamedRerouteDeclaration
   inputs [None {from=MaterialExpressionMultiply_10}]
   properties {name=DF Lines, nodeColor=<rgba 1 0 0.5391 1>}
  }
  MaterialExpressionNamedRerouteDeclaration_35 {
   class=MaterialExpressionNamedRerouteDeclaration
   inputs [None {from=MaterialExpressionSaturate_2}]
   properties {name=Section2, nodeColor=<rgba 0.1953 0 1 1>}
  }
  MaterialExpressionNamedRerouteDeclaration_36 {
   class=MaterialExpressionNamedRerouteDeclaration
   inputs [None {from=MaterialExpressionOneMinus_0}]
   properties {name=Section1, nodeColor=<rgba 0 0.4375 1 1>}
  }
  MaterialExpressionNamedRerouteDeclaration_37 {
   class=MaterialExpressionNamedRerouteDeclaration
   inputs [None {from=MaterialExpressionSaturate_3}]
   properties {name=Section3, nodeColor=<rgba 0.4297 0 1 1>}
  }
  MaterialExpressionNamedRerouteDeclaration_38 {
   class=MaterialExpressionNamedRerouteDeclaration
   inputs [None {from=MaterialExpressionSmoothStep_9}]
   properties {name=Section4, nodeColor=<rgba 0.4297 0 1 1>}
  }
  MaterialExpressionNamedRerouteDeclaration_39 {
   class=MaterialExpressionNamedRerouteDeclaration
   inputs [None {from=MaterialExpressionSaturate_4}]
   properties {name=SpectrumGradient, nodeColor=<rgba 0 1 0.3203 1>}
  }
  MaterialExpressionNamedRerouteDeclaration_4 {
   class=MaterialExpressionNamedRerouteDeclaration
   inputs [None {from=MaterialExpressionTextureSample_1, output=R}]
   properties {name=Alpha, nodeColor=<rgba 1 0.2344 0 1>}
  }
  MaterialExpressionNamedRerouteDeclaration_40 {
   class=MaterialExpressionNamedRerouteDeclaration
   inputs [None {from=MaterialExpressionSmoothStep_10}]
   properties {name=SpectrumModified, nodeColor=<rgba 0.5234 0 1 1>}
  }
  MaterialExpressionNamedRerouteDeclaration_5 {
   class=MaterialExpressionNamedRerouteDeclaration
   inputs [None {from=MaterialExpressionCollectionParameter_1}]
   properties {name=Spectrum, nodeColor=<rgba 0 0.9766 1 1>}
  }
  MaterialExpressionNamedRerouteDeclaration_6 {
   class=MaterialExpressionNamedRerouteDeclaration
   inputs [None {from=MaterialExpressionMultiply_40}]
   properties {name=DFLines, nodeColor=<rgba 0 1 0.7422 1>}
  }
  MaterialExpressionNamedRerouteDeclaration_8 {
   class=MaterialExpressionNamedRerouteDeclaration
   inputs [None {from=MaterialExpressionTextureSample_1, output=A}]
   properties {name=Mask_Original, nodeColor=<rgba 0 1 0.7891 1>}
  }
  MaterialExpressionNamedRerouteDeclaration_9 {
   class=MaterialExpressionNamedRerouteDeclaration
   inputs [None {from=MaterialExpressionPower_1}]
   properties {name=FresnelMask, nodeColor=<rgba 0.5469 1 0 1>}
  }
  MaterialExpressionNamedRerouteUsage_0 {class=MaterialExpressionNamedRerouteUsage}
  MaterialExpressionNamedRerouteUsage_1 {class=MaterialExpressionNamedRerouteUsage}
  MaterialExpressionNamedRerouteUsage_10 {class=MaterialExpressionNamedRerouteUsage}
  MaterialExpressionNamedRerouteUsage_11 {class=MaterialExpressionNamedRerouteUsage}
  MaterialExpressionNamedRerouteUsage_12 {class=MaterialExpressionNamedRerouteUsage}
  MaterialExpressionNamedRerouteUsage_13 {class=MaterialExpressionNamedRerouteUsage}
  MaterialExpressionNamedRerouteUsage_14 {class=MaterialExpressionNamedRerouteUsage}
  MaterialExpressionNamedRerouteUsage_15 {class=MaterialExpressionNamedRerouteUsage}
  MaterialExpressionNamedRerouteUsage_17 {class=MaterialExpressionNamedRerouteUsage}
  MaterialExpressionNamedRerouteUsage_19 {class=MaterialExpressionNamedRerouteUsage}
  MaterialExpressionNamedRerouteUsage_2 {class=MaterialExpressionNamedRerouteUsage}
  MaterialExpressionNamedRerouteUsage_23 {class=MaterialExpressionNamedRerouteUsage}
  MaterialExpressionNamedRerouteUsage_24 {class=MaterialExpressionNamedRerouteUsage}
  MaterialExpressionNamedRerouteUsage_26 {class=MaterialExpressionNamedRerouteUsage}
  MaterialExpressionNamedRerouteUsage_28 {class=MaterialExpressionNamedRerouteUsage}
  MaterialExpressionNamedRerouteUsage_3 {class=MaterialExpressionNamedRerouteUsage}
  MaterialExpressionNamedRerouteUsage_31 {class=MaterialExpressionNamedRerouteUsage}
  MaterialExpressionNamedRerouteUsage_38 {class=MaterialExpressionNamedRerouteUsage}
  MaterialExpressionNamedRerouteUsage_40 {class=MaterialExpressionNamedRerouteUsage}
  MaterialExpressionNamedRerouteUsage_42 {class=MaterialExpressionNamedRerouteUsage}
  MaterialExpressionNamedRerouteUsage_49 {class=MaterialExpressionNamedRerouteUsage}
  MaterialExpressionNamedRerouteUsage_5 {class=MaterialExpressionNamedRerouteUsage}
  MaterialExpressionNamedRerouteUsage_63 {class=MaterialExpressionNamedRerouteUsage}
  MaterialExpressionNamedRerouteUsage_64 {class=MaterialExpressionNamedRerouteUsage}
  MaterialExpressionNamedRerouteUsage_65 {class=MaterialExpressionNamedRerouteUsage}
  MaterialExpressionNamedRerouteUsage_67 {class=MaterialExpressionNamedRerouteUsage}
  MaterialExpressionNamedRerouteUsage_71 {class=MaterialExpressionNamedRerouteUsage}
  MaterialExpressionNamedRerouteUsage_74 {class=MaterialExpressionNamedRerouteUsage}
  MaterialExpressionNamedRerouteUsage_76 {class=MaterialExpressionNamedRerouteUsage}
  MaterialExpressionNamedRerouteUsage_8 {class=MaterialExpressionNamedRerouteUsage}
  MaterialExpressionNamedRerouteUsage_9 {class=MaterialExpressionNamedRerouteUsage}
  MaterialExpressionOneMinus_0 {class=MaterialExpressionOneMinus, inputs [None {from=MaterialExpressionSmoothStep_7}]}
  MaterialExpressionOneMinus_1 {class=MaterialExpressionOneMinus, inputs [None {from=MaterialExpressionNamedRerouteDeclaration_21}]}
  MaterialExpressionOneMinus_5 {class=MaterialExpressionOneMinus, inputs [None {from=MaterialExpressionTextureSample_0, output=R}]}
  MaterialExpressionOneMinus_6 {class=MaterialExpressionOneMinus, inputs [None {from=MaterialExpressionSmoothStep_2}]}
  MaterialExpressionPower_0 {
   class=MaterialExpressionPower
   inputs [
    Base {from=MaterialExpressionComponentMask_1}
    Exp {from=MaterialExpressionScalarParameter_11}
   ]
  }
  MaterialExpressionPower_1 {
   class=MaterialExpressionPower
   inputs [
    Base {from=MaterialExpressionOneMinus_1}
    Exp {from=MaterialExpressionScalarParameter_13}
   ]
  }
  MaterialExpressionSaturate_0 {class=MaterialExpressionSaturate, inputs [None {from=MaterialExpressionSubtract_0}]}
  MaterialExpressionSaturate_1 {class=MaterialExpressionSaturate, inputs [None {from=MaterialExpressionMultiply_26}]}
  MaterialExpressionSaturate_10 {class=MaterialExpressionSaturate, inputs [None {from=MaterialExpressionNamedRerouteUsage_9}]}
  MaterialExpressionSaturate_13 {class=MaterialExpressionSaturate, inputs [None {from=MaterialExpressionNamedRerouteUsage_63}]}
  MaterialExpressionSaturate_2 {class=MaterialExpressionSaturate, inputs [None {from=MaterialExpressionSubtract_2}]}
  MaterialExpressionSaturate_3 {class=MaterialExpressionSaturate, inputs [None {from=MaterialExpressionSubtract_5}]}
  MaterialExpressionSaturate_4 {class=MaterialExpressionSaturate, inputs [None {from=MaterialExpressionAdd_34}]}
  MaterialExpressionSaturate_9 {class=MaterialExpressionSaturate, inputs [None {from=MaterialExpressionAdd_14}]}
  MaterialExpressionScalarParameter_0 {class=MaterialExpressionScalarParameter, properties {defaultValue=6, parameterName=DFLine_Frequency}}
  MaterialExpressionScalarParameter_1 {class=MaterialExpressionScalarParameter, properties {defaultValue=-1, parameterName=DFLine_Speed}}
  MaterialExpressionScalarParameter_11 {class=MaterialExpressionScalarParameter, properties {defaultValue=1.5, parameterName=DFLine_SpectrumExponent}}
  MaterialExpressionScalarParameter_12 {class=MaterialExpressionScalarParameter, properties {defaultValue=0.2, parameterName=FresnelSoftness}}
  MaterialExpressionScalarParameter_13 {class=MaterialExpressionScalarParameter, properties {defaultValue=3, parameterName=FresnelMaskExp}}
  MaterialExpressionScalarParameter_14 {class=MaterialExpressionScalarParameter, properties {defaultValue=60, parameterName=BoundsMinOffset}}
  MaterialExpressionScalarParameter_15 {class=MaterialExpressionScalarParameter, properties {defaultValue=-30, parameterName=BoundsMaxOffset}}
  MaterialExpressionScalarParameter_19 {class=MaterialExpressionScalarParameter, properties {defaultValue=0.3, parameterName=BoostSpeed}}
  MaterialExpressionScalarParameter_20 {class=MaterialExpressionScalarParameter, properties {defaultValue=-0.1, parameterName=FresnelAmplitudeStrength}}
  MaterialExpressionScalarParameter_21 {class=MaterialExpressionScalarParameter, properties {defaultValue=0.05, parameterName=DF_InnerMaskLength}}
  MaterialExpressionScalarParameter_23 {class=MaterialExpressionScalarParameter, properties {defaultValue=0.25, parameterName=DF_OuterMaskLength}}
  MaterialExpressionScalarParameter_29 {class=MaterialExpressionScalarParameter, properties {defaultValue=0.65, parameterName=IncomingEmissiveStrength}}
  MaterialExpressionScalarParameter_3 {class=MaterialExpressionScalarParameter, properties {defaultValue=0.2, parameterName=SpectrumGradientMinThreshold}}
  MaterialExpressionScalarParameter_36 {class=MaterialExpressionScalarParameter, properties {defaultValue=0.2, parameterName=DFLine_Offset}}
  MaterialExpressionScalarParameter_4 {class=MaterialExpressionScalarParameter, properties {defaultValue=0.95, parameterName=DFLine_Thickness}}
  MaterialExpressionScalarParameter_42 {class=MaterialExpressionScalarParameter, properties {defaultValue=0.9, parameterName=BoostThreshold}}
  MaterialExpressionScalarParameter_43 {class=MaterialExpressionScalarParameter, properties {defaultValue=0.9, parameterName=BoostStrength}}
  MaterialExpressionScalarParameter_5 {class=MaterialExpressionScalarParameter, properties {defaultValue=0.01, parameterName=DFLine_Smoothness}}
  MaterialExpressionScalarParameter_6 {class=MaterialExpressionScalarParameter, properties {defaultValue=-1, parameterName=DFLine_SpeedAmpStrength}}
  MaterialExpressionScalarParameter_7 {class=MaterialExpressionScalarParameter, properties {defaultValue=0.675, parameterName=FresnelCutoff}}
  MaterialExpressionScalarParameter_9 {class=MaterialExpressionScalarParameter, properties {defaultValue=1, parameterName=DFLine_Strength}}
  MaterialExpressionSetMaterialAttributes_0 {
   class=MaterialExpressionSetMaterialAttributes
   inputs [
    MaterialAttributes {from=MaterialExpressionNamedRerouteUsage_24, output=""}
    EmissiveColor {from=MaterialExpressionLinearInterpolate_0, output=""}
   ]
   properties {attributeSetTypes [#1]}
  }
  MaterialExpressionSine_0 {class=MaterialExpressionSine, inputs [None {from=MaterialExpressionAdd_0}]}
  MaterialExpressionSine_2 {class=MaterialExpressionSine, inputs [None {from=MaterialExpressionAdd_10}]}
  MaterialExpressionSmoothStep_0 {
   class=MaterialExpressionSmoothStep
   inputs [
    Min {from=MaterialExpressionScalarParameter_4, output=""}
    Max {from=MaterialExpressionAdd_2, output=""}
    Value {from=MaterialExpressionConstantBiasScale_0, output=""}
   ]
   properties {constMax=0.6, constMin=0.58}
  }
  MaterialExpressionSmoothStep_1 {
   class=MaterialExpressionSmoothStep
   inputs [
    Min {from=MaterialExpressionScalarParameter_21, output=""}
    Max {from=MaterialExpressionAdd_22, output=""}
    Value {from=MaterialExpressionNamedRerouteUsage_5, output=""}
   ]
   properties {constMax=0.01}
  }
  MaterialExpressionSmoothStep_10 {
   class=MaterialExpressionSmoothStep
   inputs [
    Min {from=MaterialExpressionScalarParameter_3, output=""}
    Max {from="", output=""}
    Value {from=MaterialExpressionNamedRerouteUsage_31, output=""}
   ]
   properties {}
  }
  MaterialExpressionSmoothStep_11 {
   class=MaterialExpressionSmoothStep
   inputs [
    Min {from=MaterialExpressionAdd_20, output=""}
    Max {from=MaterialExpressionSaturate_9, output=""}
    Value {from=MaterialExpressionNamedRerouteUsage_38, output=""}
   ]
   properties {}
  }
  MaterialExpressionSmoothStep_2 {
   class=MaterialExpressionSmoothStep
   inputs [
    Min {from=MaterialExpressionScalarParameter_23, output=""}
    Max {from=MaterialExpressionAdd_23, output=""}
    Value {from=MaterialExpressionNamedRerouteUsage_10, output=""}
   ]
   properties {constMax=0.6, constMin=0.5}
  }
  MaterialExpressionSmoothStep_7 {
   class=MaterialExpressionSmoothStep
   inputs [
    Min {from="", output=""}
    Max {from="", output=""}
    Value {from=MaterialExpressionNamedRerouteUsage_26, output=""}
   ]
   properties {constMax=0.28, constMin=0.22}
  }
  MaterialExpressionSmoothStep_8 {
   class=MaterialExpressionSmoothStep
   inputs [
    Min {from="", output=""}
    Max {from="", output=""}
    Value {from=MaterialExpressionNamedRerouteUsage_26, output=""}
   ]
   properties {constMax=0.53, constMin=0.47}
  }
  MaterialExpressionSmoothStep_9 {
   class=MaterialExpressionSmoothStep
   inputs [
    Min {from="", output=""}
    Max {from="", output=""}
    Value {from=MaterialExpressionNamedRerouteUsage_26, output=""}
   ]
   properties {constMax=0.78, constMin=0.72}
  }
  MaterialExpressionStep_5 {
   class=MaterialExpressionStep
   inputs [
    Y {from=MaterialExpressionScalarParameter_42, output=""}
    X {from=MaterialExpressionNamedRerouteUsage_67, output=""}
   ]
   properties {}
  }
  MaterialExpressionSubtract_0 {
   class=MaterialExpressionSubtract
   inputs [
    A {from=MaterialExpressionNamedRerouteUsage_13}
    B {from=MaterialExpressionNamedRerouteUsage_15}
   ]
  }
  MaterialExpressionSubtract_1 {
   class=MaterialExpressionSubtract
   inputs [
    A {from=MaterialExpressionAdd_6}
    B {from=MaterialExpressionAdd_5}
   ]
  }
  MaterialExpressionSubtract_2 {
   class=MaterialExpressionSubtract
   inputs [
    A {from=MaterialExpressionSmoothStep_7}
    B {from=MaterialExpressionSmoothStep_8}
   ]
  }
  MaterialExpressionSubtract_4 {
   class=MaterialExpressionSubtract
   inputs [
    A {from=MaterialExpressionTransformPosition_4}
    B {from=MaterialExpressionAdd_5}
   ]
  }
  MaterialExpressionSubtract_5 {
   class=MaterialExpressionSubtract
   inputs [
    A {from=MaterialExpressionSmoothStep_8}
    B {from=MaterialExpressionSmoothStep_9}
   ]
  }
  MaterialExpressionTextureSample_0 {
   class=MaterialExpressionTextureSample
   inputs [
    UVs {from=""}
    Tex {from=""}
    Apply View MipBias {from=""}
   ]
   properties {
    samplerType=SAMPLERTYPE_LinearGrayscale
    texture=/BRCosmetics/Characters/Player/Female/Medium/Bodies/F_MED_PhoneCharger/Skins/Cable/Textures/T_PhoneCharger_Cable_Body_Reactive_DF
   }
  }
  MaterialExpressionTextureSample_1 {
   class=MaterialExpressionTextureSample
   inputs [
    UVs {from=""}
    Tex {from=""}
    Apply View MipBias {from=""}
   ]
   properties {
    samplerType=SAMPLERTYPE_Masks
    texture=/BRCosmetics/Characters/Player/Female/Medium/Bodies/F_MED_PhoneCharger/Skins/Cable/Textures/T_PhoneCharger_Cable_Body_Reactive_M
   }
  }
  MaterialExpressionTime_0 {class=MaterialExpressionTime}
  MaterialExpressionTime_1 {class=MaterialExpressionTime}
  MaterialExpressionTransformPosition_4 {
   class=MaterialExpressionTransformPosition
   inputs [None {from=MaterialExpressionWorldPosition_4, output=XYZ}]
   properties {transformSourceType=TRANSFORMPOSSOURCE_World, transformType=TRANSFORMPOSSOURCE_Instance}
  }
  MaterialExpressionVectorParameter_0 {class=MaterialExpressionVectorParameter, properties {parameterName=EmissiveColor, defaultValue=<rgba 0.0409 0.9216 0 1>}}
  MaterialExpressionWorldPosition_4 {class=MaterialExpressionWorldPosition, inputs [], properties {worldPositionShaderOffset=WPT_ExcludeAllShaderOffsets}}
 }}
